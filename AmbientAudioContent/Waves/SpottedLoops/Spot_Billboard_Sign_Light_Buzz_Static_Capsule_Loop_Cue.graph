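# Looping Sound Cue for a spotted ambient billboard sign light buzz and static.
# A Mixer layers a random pick of four hologram loops, pitched up an octave by a Modulator, with a random pick of two neon light loops.
# An outer Modulator varies the mix.
# Capsule attenuation with low-pass filtering and listener focus keeps it local, with a two-instance concurrency cap and a reverb submix send.
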SoundCue /AmbientAudioContent/Waves/SpottedLoops/Spot_Billboard_Sign_Light_Buzz_Static_Capsule_Loop_Cue
root=Modulator_0
nodes [
 {id=Mixer_0, children [Modulator_1,Random_1], properties {inputVolume [1,1]}}
 {id=Modulator_0, children [Mixer_0], properties {pitchMin=1}}
 {id=Modulator_1, children [Random_0], properties {pitchMax=2, pitchMin=2}}
 {id=Random_0, children [WavePlayer_1,WavePlayer_2,WavePlayer_3,WavePlayer_4], properties {weights [1,1,1,1]}}
 {id=Random_1, children [WavePlayer_5,WavePlayer_6], properties {weights [1,1]}}
 {
  id=WavePlayer_1
  properties {bLooping=true, soundWaveAssetPtr=/Game/Sounds/Foley_Loot/Containers/Keys_And_Locks/Keys_And_Locks_HologramLoop_01}
 }
 {
  id=WavePlayer_2
  properties {bLooping=true, soundWaveAssetPtr=/Game/Sounds/Foley_Loot/Containers/Keys_And_Locks/Keys_And_Locks_HologramLoop_02}
 }
 {
  id=WavePlayer_3
  properties {bLooping=true, soundWaveAssetPtr=/Game/Sounds/Foley_Loot/Containers/Keys_And_Locks/Keys_And_Locks_HologramLoop_03}
 }
 {
  id=WavePlayer_4
  properties {bLooping=true, soundWaveAssetPtr=/Game/Sounds/Foley_Loot/Containers/Keys_And_Locks/Keys_And_Locks_HologramLoop_04}
 }
 {id=WavePlayer_5, properties {bLooping=true, soundWaveAssetPtr=&/Spot_City_NeonLight_02}}
 {id=WavePlayer_6, properties {bLooping=true, soundWaveAssetPtr=&/Spot_City_NeonLight_01}}
]
properties {
 bOverrideAttenuation=true
 bOverrideConcurrency=true
 maxDistance=1050
 soundClassObject=&Fort_Gadgets_Items
 volumeMultiplier=0.25
 attenuationOverrides {
  attenuationShape=Capsule
  bAttenuateWithLPF=true
  bEnableListenerFocus=true
  falloffDistance=650
  hPFFrequencyAtMax=1000
  hPFFrequencyAtMin=100
  lPFRadiusMax=2000
  lPFRadiusMin=300
  nonFocusVolumeAttenuation=0.4
  attenuationShapeExtents {y=250}
 }
 concurrencyOverrides {maxCount=2}
 concurrencySet [&Weapon_Drop_StopOldest]
 soundSubmixSends [{minSendLevel=0.25, sendLevelControlMethod=Linear, soundSubmix=&Fortnite_Reverb_Default}]
}
